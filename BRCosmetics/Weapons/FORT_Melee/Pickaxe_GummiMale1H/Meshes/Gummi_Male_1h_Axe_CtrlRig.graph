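# Control Rig for the Gummi Male one-handed pickaxe that adds procedural wobble to its body and tentacle bones.
# One execution chain first runs three Bone Harmonics units: two on alternating upper-body bones with differing wave settings, one on the lower body.
# It then runs Chain Harmonics on eleven tentacle chains (a to k), using only the wave with falloff curves; pendulum and reach are off.
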ControlRigBlueprint /BRCosmetics/Weapons/FORT_Melee/Pickaxe_GummiMale1H/Meshes/Gummi_Male_1h_Axe_CtrlRig
graphs [
 RigVMFunctionLibrary {links [], nodes []}
 RigVMModel {nodes [
   n1 = &RigUnit_BeginExecution()
   /Script/ControlRig.RigUnit_BoneHarmonics(
    Bones=((Bone="dyn_up_body_01",Ratio=0),(Bone="dyn_up_body_03",Ratio=0),(Bone="dyn_up_body_05",Ratio=0))
    RotationOrder=YZX
    WaveAmplitude=(X=0,Y=3,Z=2)
    WaveEase=Linear
    WaveFrequency=(X=1,Y=4,Z=2)
    WaveMaximum=1
    WaveMinimum=0.5
    WaveNoise=(X=0,Y=0.1,Z=0.1)
    WaveOffset=(X=0,Y=0,Z=0)
    WaveSpeed=(X=1,Y=1.5,Z=1)
    bPropagateToChildren=True
    ExecutePin=@n1.ExecutePin
   ) {is=struct RigVMUnitNode}
   > &RigUnit_BoneHarmonics(
    Bones=((Bone="dyn_up_body_04",Ratio=1),(Bone="dyn_up_body_02",Ratio=0))
    RotationOrder=YZX
    WaveAmplitude=(X=0,Y=2,Z=3)
    WaveEase=Linear
    WaveFrequency=(X=1,Y=4,Z=2)
    WaveMaximum=1
    WaveMinimum=0.5
    WaveNoise=(X=0,Y=0.1,Z=0.1)
    WaveOffset=(X=0,Y=2,Z=2)
    WaveSpeed=(X=1,Y=1,Z=1.5)
    bPropagateToChildren=True
   )
   > &RigUnit_BoneHarmonics(
    Bones=((Bone="dyn_lw_body_01",Ratio=0))
    RotationOrder=YZX
    WaveAmplitude=(X=0,Y=15,Z=10)
    WaveEase=Linear
    WaveFrequency=(X=1,Y=2,Z=2)
    WaveMaximum=1
    WaveMinimum=0.5
    WaveNoise=(X=0,Y=0.1,Z=0.1)
    WaveOffset=(X=0,Y=1,Z=1)
    WaveSpeed=(X=1,Y=1,Z=1.5)
    bPropagateToChildren=True
   )
   > /Script/ControlRig.RigUnit_ChainHarmonics(
    ChainRoot=dyn_tentacle_a_01
    DrawWorldOffset=(Rotation=(X=0,Y=0,Z=0,W=1),Translation=(X=0,Y=0,Z=0),Scale3D=(X=1,Y=1,Z=1))
    Pendulum=(bEnabled=False,PendulumStiffness=2,PendulumGravity=(X=0,Y=0,Z=-0.98),PendulumBlend=0.75,PendulumDrag=0.98,PendulumMinimum=0,PendulumMaximum=0.1,PendulumEase=Linear,UnwindAxis=(X=0,Y=1,Z=0),UnwindMinimum=0.2,UnwindMaximum=0.05)
    Reach=(bEnabled=False,ReachTarget=(X=0,Y=0,Z=0),ReachAxis=(X=1,Y=0,Z=0),ReachMinimum=0,ReachMaximum=0,ReachEase=Linear)
    Speed=(X=1,Y=1.5,Z=1)
    Wave=(bEnabled=True,WaveFrequency=(X=1,Y=1,Z=2),WaveAmplitude=(X=0,Y=0.35,Z=0.4),WaveOffset=(X=0,Y=0.3,Z=0.3),WaveNoise=(X=0,Y=0.2,Z=0.2),WaveMinimum=0,WaveMaximum=1,WaveEase=QuadraticEaseInOut)
    WaveCurve=(EditorCurveData=(Keys=((Value=0.8),(Time=1,Value=1)),PreInfinityExtrap=RCCE_Constant,PostInfinityExtrap=RCCE_Constant,DefaultValue=340282346638528859811704183484516925440))
    bDrawDebug=false
   ) {is=struct RigVMUnitNode}
   > &RigUnit_ChainHarmonics(
    ChainRoot=dyn_tentacle_b_01
    DrawWorldOffset=(Rotation=(X=0,Y=0,Z=0,W=1),Translation=(X=0,Y=0,Z=0),Scale3D=(X=1,Y=1,Z=1))
    Pendulum=(bEnabled=False,PendulumStiffness=2,PendulumGravity=(X=0,Y=0,Z=-0.98),PendulumBlend=0.75,PendulumDrag=0.98,PendulumMinimum=0,PendulumMaximum=0.1,PendulumEase=Linear,UnwindAxis=(X=0,Y=1,Z=0),UnwindMinimum=0.2,UnwindMaximum=0.05)
    Reach=(bEnabled=False,ReachTarget=(X=0,Y=0,Z=0),ReachAxis=(X=1,Y=0,Z=0),ReachMinimum=0,ReachMaximum=0,ReachEase=Linear)
    Speed=(X=1,Y=1.5,Z=1)
    Wave=(bEnabled=True,WaveFrequency=(X=1,Y=2,Z=3),WaveAmplitude=(X=0,Y=0.35,Z=0.4),WaveOffset=(X=0,Y=2,Z=2),WaveNoise=(X=0,Y=0.2,Z=0.2),WaveMinimum=0,WaveMaximum=1,WaveEase=QuadraticEaseInOut)
    WaveCurve=(EditorCurveData=(Keys=((Value=0.8),(Time=1,Value=1)),PreInfinityExtrap=RCCE_Constant,PostInfinityExtrap=RCCE_Constant,DefaultValue=340282346638528859811704183484516925440))
    bDrawDebug=false
   )
   > &RigUnit_ChainHarmonics(
    ChainRoot=dyn_tentacle_c_01
    DrawWorldOffset=(Rotation=(X=0,Y=0,Z=0,W=1),Translation=(X=0,Y=0,Z=0),Scale3D=(X=1,Y=1,Z=1))
    Pendulum=(bEnabled=False,PendulumStiffness=2,PendulumGravity=(X=0,Y=0,Z=-0.98),PendulumBlend=0.75,PendulumDrag=0.98,PendulumMinimum=0,PendulumMaximum=0.1,PendulumEase=Linear,UnwindAxis=(X=0,Y=1,Z=0),UnwindMinimum=0.2,UnwindMaximum=0.05)
    Reach=(bEnabled=False,ReachTarget=(X=0,Y=0,Z=0),ReachAxis=(X=1,Y=0,Z=0),ReachMinimum=0,ReachMaximum=0,ReachEase=Linear)
    Speed=(X=2,Y=0.5,Z=0.5)
    Wave=(bEnabled=True,WaveFrequency=(X=1,Y=1,Z=2),WaveAmplitude=(X=0,Y=0.3,Z=0.3),WaveOffset=(X=0,Y=2,Z=1),WaveNoise=(X=0,Y=0.5,Z=0.5),WaveMinimum=0,WaveMaximum=1,WaveEase=Linear)
    WaveCurve=(EditorCurveData=(Keys=((Time=0.0452,Value=0.7355),(Time=0.3862,Value=0.785),(Time=0.8152,Value=0.9115),(Time=1,Value=1)),PreInfinityExtrap=RCCE_Constant,PostInfinityExtrap=RCCE_Constant,DefaultValue=340282346638528859811704183484516925440))
    bDrawDebug=false
   )
   > &RigUnit_ChainHarmonics(
    ChainRoot=dyn_tentacle_d_01
    DrawWorldOffset=(Rotation=(X=0,Y=0,Z=0,W=1),Translation=(X=0,Y=0,Z=0),Scale3D=(X=1,Y=1,Z=1))
    Pendulum=(bEnabled=False,PendulumStiffness=2,PendulumGravity=(X=0,Y=0,Z=-0.98),PendulumBlend=0.75,PendulumDrag=0.98,PendulumMinimum=0,PendulumMaximum=0.1,PendulumEase=Linear,UnwindAxis=(X=0,Y=1,Z=0),UnwindMinimum=0.2,UnwindMaximum=0.05)
    Reach=(bEnabled=False,ReachTarget=(X=0,Y=0,Z=0),ReachAxis=(X=1,Y=0,Z=0),ReachMinimum=0,ReachMaximum=0,ReachEase=Linear)
    Speed=(X=2,Y=0.5,Z=0.5)
    Wave=(bEnabled=True,WaveFrequency=(X=1,Y=1,Z=2),WaveAmplitude=(X=0,Y=0.3,Z=0.3),WaveOffset=(X=0,Y=2,Z=1),WaveNoise=(X=0,Y=0.5,Z=0.5),WaveMinimum=0,WaveMaximum=1,WaveEase=Linear)
    WaveCurve=(EditorCurveData=(Keys=((Time=0.0452,Value=0.7355),(Time=0.3862,Value=0.785),(Time=0.8152,Value=0.9115),(Time=1,Value=1)),PreInfinityExtrap=RCCE_Constant,PostInfinityExtrap=RCCE_Constant,DefaultValue=340282346638528859811704183484516925440))
    bDrawDebug=false
   )
   > &RigUnit_ChainHarmonics(
    ChainRoot=dyn_tentacle_e_01
    DrawWorldOffset=(Rotation=(X=0,Y=0,Z=0,W=1),Translation=(X=0,Y=0,Z=0),Scale3D=(X=1,Y=1,Z=1))
    Pendulum=(bEnabled=False,PendulumStiffness=2,PendulumGravity=(X=0,Y=0,Z=-0.98),PendulumBlend=0.75,PendulumDrag=0.98,PendulumMinimum=0,PendulumMaximum=0.1,PendulumEase=Linear,UnwindAxis=(X=0,Y=1,Z=0),UnwindMinimum=0.2,UnwindMaximum=0.05)
    Reach=(bEnabled=False,ReachTarget=(X=0,Y=0,Z=0),ReachAxis=(X=1,Y=0,Z=0),ReachMinimum=0,ReachMaximum=0,ReachEase=Linear)
    Speed=(X=2,Y=0.5,Z=0.5)
    Wave=(bEnabled=True,WaveFrequency=(X=1,Y=1,Z=2),WaveAmplitude=(X=0,Y=0.3,Z=0.3),WaveOffset=(X=0,Y=2,Z=1),WaveNoise=(X=0,Y=0.5,Z=0.5),WaveMinimum=0,WaveMaximum=1,WaveEase=Linear)
    WaveCurve=(EditorCurveData=(Keys=((Time=0.0452,Value=0.7355),(Time=0.3862,Value=0.785),(Time=0.8152,Value=0.9115),(Time=1,Value=1)),PreInfinityExtrap=RCCE_Constant,PostInfinityExtrap=RCCE_Constant,DefaultValue=340282346638528859811704183484516925440))
    bDrawDebug=false
   )
   > &RigUnit_ChainHarmonics(
    ChainRoot=dyn_tentacle_f_01
    DrawWorldOffset=(Rotation=(X=0,Y=0,Z=0,W=1),Translation=(X=0,Y=0,Z=0),Scale3D=(X=1,Y=1,Z=1))
    Pendulum=(bEnabled=False,PendulumStiffness=2,PendulumGravity=(X=0,Y=0,Z=-0.98),PendulumBlend=0.75,PendulumDrag=0.98,PendulumMinimum=0,PendulumMaximum=0.1,PendulumEase=Linear,UnwindAxis=(X=0,Y=1,Z=0),UnwindMinimum=0.2,UnwindMaximum=0.05)
    Reach=(bEnabled=False,ReachTarget=(X=0,Y=0,Z=0),ReachAxis=(X=1,Y=0,Z=0),ReachMinimum=0,ReachMaximum=0,ReachEase=Linear)
    Speed=(X=2,Y=0.6,Z=0.7)
    Wave=(bEnabled=True,WaveFrequency=(X=1,Y=2,Z=4),WaveAmplitude=(X=0,Y=0.3,Z=0.5),WaveOffset=(X=0,Y=2,Z=1),WaveNoise=(X=0,Y=0.5,Z=0.5),WaveMinimum=0,WaveMaximum=1,WaveEase=Linear)
    WaveCurve=(EditorCurveData=(Keys=((Time=0.0452,Value=0.7355),(Time=1,Value=1)),PreInfinityExtrap=RCCE_Constant,PostInfinityExtrap=RCCE_Constant,DefaultValue=340282346638528859811704183484516925440))
    bDrawDebug=false
   )
   > &RigUnit_ChainHarmonics(
    ChainRoot=dyn_tentacle_g_01
    DrawWorldOffset=(Rotation=(X=0,Y=0,Z=0,W=1),Translation=(X=0,Y=0,Z=0),Scale3D=(X=1,Y=1,Z=1))
    Pendulum=(bEnabled=False,PendulumStiffness=2,PendulumGravity=(X=0,Y=0,Z=-0.98),PendulumBlend=0.75,PendulumDrag=0.98,PendulumMinimum=0,PendulumMaximum=0.1,PendulumEase=Linear,UnwindAxis=(X=0,Y=1,Z=0),UnwindMinimum=0.2,UnwindMaximum=0.05)
    Reach=(bEnabled=False,ReachTarget=(X=0,Y=0,Z=0),ReachAxis=(X=1,Y=0,Z=0),ReachMinimum=0,ReachMaximum=0,ReachEase=Linear)
    Speed=(X=2,Y=0.5,Z=0.5)
    Wave=(bEnabled=True,WaveFrequency=(X=1,Y=1,Z=2),WaveAmplitude=(X=0,Y=0.3,Z=0.3),WaveOffset=(X=0,Y=2,Z=1),WaveNoise=(X=0,Y=0.5,Z=0.5),WaveMinimum=0,WaveMaximum=1,WaveEase=Linear)
    WaveCurve=(EditorCurveData=(Keys=((Time=0.0452,Value=0.7355),(Time=0.3862,Value=0.785),(Time=0.8152,Value=0.9115),(Time=1,Value=1)),PreInfinityExtrap=RCCE_Constant,PostInfinityExtrap=RCCE_Constant,DefaultValue=340282346638528859811704183484516925440))
    bDrawDebug=false
   )
   > &RigUnit_ChainHarmonics(
    ChainRoot=dyn_tentacle_h_01
    DrawWorldOffset=(Rotation=(X=0,Y=0,Z=0,W=1),Translation=(X=0,Y=0,Z=0),Scale3D=(X=1,Y=1,Z=1))
    Pendulum=(bEnabled=False,PendulumStiffness=2,PendulumGravity=(X=0,Y=0,Z=-0.98),PendulumBlend=0.75,PendulumDrag=0.98,PendulumMinimum=0,PendulumMaximum=0.1,PendulumEase=Linear,UnwindAxis=(X=0,Y=1,Z=0),UnwindMinimum=0.2,UnwindMaximum=0.05)
    Reach=(bEnabled=False,ReachTarget=(X=0,Y=0,Z=0),ReachAxis=(X=1,Y=0,Z=0),ReachMinimum=0,ReachMaximum=0,ReachEase=Linear)
    Speed=(X=2,Y=0.5,Z=0.5)
    Wave=(bEnabled=True,WaveFrequency=(X=1,Y=1,Z=2),WaveAmplitude=(X=0,Y=0.3,Z=0.3),WaveOffset=(X=0,Y=2,Z=1),WaveNoise=(X=0,Y=0.5,Z=0.5),WaveMinimum=0,WaveMaximum=1,WaveEase=Linear)
    WaveCurve=(EditorCurveData=(Keys=((Time=0.0452,Value=0.7355),(Time=0.3862,Value=0.785),(Time=0.8152,Value=0.9115),(Time=1,Value=1)),PreInfinityExtrap=RCCE_Constant,PostInfinityExtrap=RCCE_Constant,DefaultValue=340282346638528859811704183484516925440))
    bDrawDebug=false
   )
   > &RigUnit_ChainHarmonics(
    ChainRoot=dyn_tentacle_i_01
    DrawWorldOffset=(Rotation=(X=0,Y=0,Z=0,W=1),Translation=(X=0,Y=0,Z=0),Scale3D=(X=1,Y=1,Z=1))
    Pendulum=(bEnabled=False,PendulumStiffness=2,PendulumGravity=(X=0,Y=0,Z=-0.98),PendulumBlend=0.75,PendulumDrag=0.98,PendulumMinimum=0,PendulumMaximum=0.1,PendulumEase=Linear,UnwindAxis=(X=0,Y=1,Z=0),UnwindMinimum=0.2,UnwindMaximum=0.05)
    Reach=(bEnabled=False,ReachTarget=(X=0,Y=0,Z=0),ReachAxis=(X=1,Y=0,Z=0),ReachMinimum=0,ReachMaximum=0,ReachEase=Linear)
    Speed=(X=2,Y=0.5,Z=0.5)
    Wave=(bEnabled=True,WaveFrequency=(X=1,Y=1,Z=2),WaveAmplitude=(X=0,Y=0.3,Z=0.3),WaveOffset=(X=0,Y=2,Z=1),WaveNoise=(X=0,Y=0.5,Z=0.5),WaveMinimum=0,WaveMaximum=1,WaveEase=Linear)
    WaveCurve=(EditorCurveData=(Keys=((Time=0.0452,Value=0.7355),(Time=0.3862,Value=0.785),(Time=0.8152,Value=0.9115),(Time=1,Value=1)),PreInfinityExtrap=RCCE_Constant,PostInfinityExtrap=RCCE_Constant,DefaultValue=340282346638528859811704183484516925440))
    bDrawDebug=false
   )
   > &RigUnit_ChainHarmonics(
    ChainRoot=dyn_tentacle_j_01
    DrawWorldOffset=(Rotation=(X=0,Y=0,Z=0,W=1),Translation=(X=0,Y=0,Z=0),Scale3D=(X=1,Y=1,Z=1))
    Pendulum=(bEnabled=False,PendulumStiffness=2,PendulumGravity=(X=0,Y=0,Z=-0.98),PendulumBlend=0.75,PendulumDrag=0.98,PendulumMinimum=0,PendulumMaximum=0.1,PendulumEase=Linear,UnwindAxis=(X=0,Y=1,Z=0),UnwindMinimum=0.2,UnwindMaximum=0.05)
    Reach=(bEnabled=False,ReachTarget=(X=0,Y=0,Z=0),ReachAxis=(X=1,Y=0,Z=0),ReachMinimum=0,ReachMaximum=0,ReachEase=Linear)
    Speed=(X=2,Y=0.8,Z=0.8)
    Wave=(bEnabled=True,WaveFrequency=(X=1,Y=1,Z=2),WaveAmplitude=(X=0,Y=0.5,Z=0.5),WaveOffset=(X=0,Y=2,Z=1),WaveNoise=(X=0,Y=0.5,Z=0.5),WaveMinimum=0,WaveMaximum=1,WaveEase=Linear)
    WaveCurve=(EditorCurveData=(Keys=((Time=0.0452,Value=0.7355),(Time=0.3862,Value=0.785),(Time=0.8152,Value=0.9115),(Time=1,Value=1)),PreInfinityExtrap=RCCE_Constant,PostInfinityExtrap=RCCE_Constant,DefaultValue=340282346638528859811704183484516925440))
    bDrawDebug=false
   )
   > &RigUnit_ChainHarmonics(
    ChainRoot=dyn_tentacle_k_01
    DrawWorldOffset=(Rotation=(X=0,Y=0,Z=0,W=1),Translation=(X=0,Y=0,Z=0),Scale3D=(X=1,Y=1,Z=1))
    Pendulum=(bEnabled=False,PendulumStiffness=2,PendulumGravity=(X=0,Y=0,Z=-0.98),PendulumBlend=0.75,PendulumDrag=0.98,PendulumMinimum=0,PendulumMaximum=0.1,PendulumEase=Linear,UnwindAxis=(X=0,Y=1,Z=0),UnwindMinimum=0.2,UnwindMaximum=0.05)
    Reach=(bEnabled=False,ReachTarget=(X=0,Y=0,Z=0),ReachAxis=(X=1,Y=0,Z=0),ReachMinimum=0,ReachMaximum=0,ReachEase=Linear)
    Speed=(X=2,Y=0.5,Z=0.5)
    Wave=(bEnabled=True,WaveFrequency=(X=1,Y=1,Z=2),WaveAmplitude=(X=0,Y=0.3,Z=0.3),WaveOffset=(X=0,Y=2,Z=1),WaveNoise=(X=0,Y=0.5,Z=0.5),WaveMinimum=0,WaveMaximum=1,WaveEase=Linear)
    WaveCurve=(EditorCurveData=(Keys=((Time=0.0452,Value=0.7355),(Time=0.3862,Value=0.785),(Time=0.8152,Value=0.9115),(Time=1,Value=1)),PreInfinityExtrap=RCCE_Constant,PostInfinityExtrap=RCCE_Constant,DefaultValue=340282346638528859811704183484516925440))
    bDrawDebug=false
   )
  ]}
]
hierarchy [
 b_AR_Root {loc=<0 0 0>, rot=<0 0 0.7071 0.7071>}
 Melee_TwoHanded_Handle {parent=b_AR_Root, loc=<8.1477 -0.008 -29.0793>, rot=<0 0 0.7071 -0.7071>}
 dyn_body_main {parent=Melee_TwoHanded_Handle, loc=<0.0971 -8.1795 45.6899>, rot=<0 0 0 -1>}
 dyn_up_body_01 {parent=dyn_body_main, loc=<-0.5432 0.0318 -7.1422>, rot=<0.7497 0 0.6618 0>}
 dyn_up_body_02 {parent=dyn_up_body_01, loc=<8.3987 0 0>, rot=<0 -0.0394 0 -0.9992>}
 dyn_up_body_03 {parent=dyn_up_body_02, loc=<8.2672 0 0>, rot=<0 -0.1032 0 -0.9947>}
 dyn_tentacle_k_01 {parent=dyn_up_body_03, loc=<1.0498 0.2979 4.6279>, rot=<0.2314 0.0163 0.9703 0.0686>}
 dyn_tentacle_k_02 {parent=dyn_tentacle_k_01, loc=<4.2763 0 0>, rot=<0.0094 -0.1012 -0.0478 -0.9937>}
 dyn_up_body_04 {parent=dyn_up_body_03, loc=<8.4995 0 0>, rot=<0 -0.158 0 -0.9874>}
 dyn_tentacle_i_01 {parent=dyn_up_body_04, loc=<-2.6068 -1.2991 2.2749>, rot=<0.2998 -0.0373 0.946 -0.1179>}
 dyn_tentacle_i_02 {parent=dyn_tentacle_i_01, loc=<3.2397 -0.1195 -0.4697>, rot=<-0.0417 -0.2592 0.0853 -0.9611>}
 dyn_tentacle_j_01 {parent=dyn_up_body_04, loc=<-3.0254 0.591 2.4663>, rot=<0.3595 -0.0151 0.9322 -0.039>}
 dyn_tentacle_j_02 {parent=dyn_tentacle_j_01, loc=<3.5067 0 0>, rot=<-0.0089 -0.107 0.001 -0.9942>}
 dyn_tentacle_j_03 {parent=dyn_tentacle_j_02, loc=<3.3513 0 0>, rot=<0 -0.0497 0 -0.9988>}
 dyn_tentacle_h_01 {parent=dyn_up_body_04, loc=<-0.8653 1.2577 2.881>, rot=<0.3755 0.0541 0.9158 0.1319>}
 dyn_tentacle_h_02 {parent=dyn_tentacle_h_01, loc=<2.9032 0.0286 -0.1501>, rot=<0.0729 -0.3459 -0.0732 -0.9326>}
 dyn_tentacle_g_01 {parent=dyn_up_body_04, loc=<0.5199 -1.0314 3.5236>, rot=<0.3244 -0.0358 0.9396 -0.1037>}
 dyn_tentacle_g_02 {parent=dyn_tentacle_g_01, loc=<3.1632 0 0>, rot=<-0.0361 -0.2048 0.0442 -0.9771>}
 dyn_tentacle_f_01 {parent=dyn_up_body_04, loc=<1.9811 -0.0108 4.2817>, rot=<0.3724 0.0268 0.9253 0.0665>}
 dyn_tentacle_f_02 {parent=dyn_tentacle_f_01, loc=<4.3012 0 0>, rot=<0.0296 -0.2605 -0.0294 -0.9646>}
 dyn_up_body_05 {parent=dyn_up_body_04, loc=<6.6215 0 0>, rot=<0 -0.245 0 -0.9695>}
 dyn_tentacle_e_01 {parent=dyn_up_body_05, loc=<-4.1152 1.0223 2.972>, rot=<0.1396 0.0185 0.9814 0.1303>}
 dyn_tentacle_e_02 {parent=dyn_tentacle_e_01, loc=<2.6274 0 0>, rot=<0.0095 -0.0335 0.0128 -0.9993>}
 dyn_tentacle_d_01 {parent=dyn_up_body_05, loc=<-3.584 0.043 4.2026>, rot=<0.2641 0.0002 0.9645 0.0006>}
 dyn_tentacle_d_02 {parent=dyn_tentacle_d_01, loc=<4.1332 0 0>, rot=<-0.0008 -0.3257 -0.0036 -0.9455>}
 dyn_tentacle_c_01 {parent=dyn_up_body_05, loc=<-1.6879 -1.6071 4.3711>, rot=<0.5621 -0.0482 0.8226 -0.0706>}
 dyn_tentacle_c_02 {parent=dyn_tentacle_c_01, loc=<2.589 0.0414 -0.1052>, rot=<-0.0397 -0.1776 -0.0477 -0.9821>}
 dyn_tentacle_b_01 {parent=dyn_up_body_05, loc=<2.3843 -0.9037 5.1815>, rot=<0.6981 -0.1002 0.7018 -0.1007>}
 dyn_tentacle_b_02 {parent=dyn_tentacle_b_01, loc=<3.2278 0 0>, rot=<0.0165 0.084 0.0908 -0.9922>}
 dyn_tentacle_b_03 {parent=dyn_tentacle_b_02, loc=<2.4743 0 0>, rot=<0.0074 0.1692 0.058 -0.9838>}
 dyn_tentacle_a_01 {parent=dyn_up_body_05, loc=<4.9932 1.0409 4.5937>, rot=<0.8264 0.1048 0.5488 0.0696>}
 dyn_tentacle_a_02 {parent=dyn_tentacle_a_01, loc=<3.1778 0 0>, rot=<0.0012 -0.0066 -0.0709 -0.9975>}
 dyn_tentacle_a_03 {parent=dyn_tentacle_a_02, loc=<2.5716 0 0>, rot=<-0.0068 0.1425 -0.062 -0.9878>}
 dyn_lw_body_01 {parent=dyn_body_main, loc=<0.922 0.0318 -24.9247>, rot=<0 -0.4618 0 -0.887>}
]
