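# Physics asset for the dynamic bones of the Gummi Male one-handed axe pickaxe cosmetic, defining colliders and joints for its tentacles.
# A main body bone carries two angled capsules and anchors eleven tentacle chains, lettered a to k.
# Each chain has two or three small capsule bodies, jointed parent to child back to the body.
# All joints share the same moderate swing and twist angle settings.
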PhysicsAsset /BRCosmetics/Weapons/FORT_Melee/Pickaxe_GummiMale1H/Meshes/Gummi_Male_1h_Axe_Dyn
bodies {
 dyn_body_main [
  {shapeType=Capsule, length=8.0194, radius=4.3395, center=<-5.2992 0 26.9954>, rotation=<rot 49.9997 0 0>}
  {shapeType=Capsule, length=20.9705, radius=4.3395, center=<-0.2212 0 15.0175>, rotation=<rot 14.9999 0 0>}
 ]
 dyn_tentacle_a_01 [{shapeType=Capsule, length=2, radius=1, center=<1 0 0>, rotation=<rot 90 0 0>}]
 dyn_tentacle_a_02 [{shapeType=Capsule, length=2, radius=1, center=<1 0 0>, rotation=<rot 90 0 0>}]
 dyn_tentacle_a_03 [{shapeType=Capsule, length=2, radius=1, center=<1 0 0>, rotation=<rot 90 0 0>}]
 dyn_tentacle_b_01 [{shapeType=Capsule, length=2, radius=1, center=<1 0 0>, rotation=<rot 90 0 0>}]
 dyn_tentacle_b_02 [{shapeType=Capsule, length=2, radius=1, center=<1 0 0>, rotation=<rot 90 0 0>}]
 dyn_tentacle_b_03 [{shapeType=Capsule, length=2, radius=1, center=<1 0 0>, rotation=<rot 90 0 0>}]
 dyn_tentacle_c_01 [{shapeType=Capsule, length=2, radius=1, center=<1 0 0>, rotation=<rot 90 0 0>}]
 dyn_tentacle_c_02 [{shapeType=Capsule, length=2.4136, radius=1, center=<1.2125 0 0>, rotation=<rot 90 0 0>}]
 dyn_tentacle_d_01 [{shapeType=Capsule, length=3.2592, radius=1, center=<2.1002 0 0>, rotation=<rot 90 0 0>}]
 dyn_tentacle_d_02 [{shapeType=Capsule, length=2.3754, radius=1, center=<1.554 0 0>, rotation=<rot 90 0 0>}]
 dyn_tentacle_e_01 [{shapeType=Capsule, length=2, radius=1, center=<1 0 0>, rotation=<rot 90 0 0>}]
 dyn_tentacle_e_02 [{shapeType=Capsule, length=2, radius=1, center=<1 0 0>, rotation=<rot 90 0 0>}]
 dyn_tentacle_f_01 [{shapeType=Capsule, length=3.2577, radius=0.8869, center=<2.0061 -0.0001 0>, rotation=<rot 90 0 0>}]
 dyn_tentacle_f_02 [{shapeType=Capsule, length=4.061, radius=1.1789, center=<2.0501 0 0>, rotation=<rot 90 0 0>}]
 dyn_tentacle_g_01 [{shapeType=Capsule, length=2, radius=1, center=<1 0 0>, rotation=<rot 90 0 0>}]
 dyn_tentacle_g_02 [{shapeType=Capsule, length=2, radius=1, center=<1 0 0>, rotation=<rot 90 0 0>}]
 dyn_tentacle_h_01 [{shapeType=Capsule, length=2, radius=1, center=<1 0 0>, rotation=<rot 90 0 0>}]
 dyn_tentacle_h_02 [{shapeType=Capsule, length=3.7104, radius=1, center=<1.6357 0 0>, rotation=<rot 90 0 0>}]
 dyn_tentacle_i_01 [{shapeType=Capsule, length=2, radius=1, center=<1 0 0>, rotation=<rot 90 0 0>}]
 dyn_tentacle_i_02 [{shapeType=Capsule, length=2, radius=1, center=<1 0 0>, rotation=<rot 90 0 0>}]
 dyn_tentacle_j_01 [{shapeType=Capsule, length=2, radius=1, center=<1 0 0>, rotation=<rot 90 0 0>}]
 dyn_tentacle_j_02 [{shapeType=Capsule, length=2, radius=1, center=<1 0 0>, rotation=<rot 90 0 0>}]
 dyn_tentacle_j_03 [{shapeType=Capsule, length=1.391, radius=1, center=<0.7084 0 0>, rotation=<rot 90 0 0>}]
 dyn_tentacle_k_01 [{shapeType=Capsule, length=3.49, radius=1, center=<1.9219 0 0>, rotation=<rot 90 0 0>}]
 dyn_tentacle_k_02 [{shapeType=Capsule, length=2.5092, radius=1, center=<1.1 0 0>, rotation=<rot 90 0 0>}]
}
constraints [
 {bone1Name=dyn_tentacle_a_01, bone2Name=dyn_body_main, swing1LimitDegrees=45, swing2LimitDegrees=45, twistLimitDegrees=45}
 {bone1Name=dyn_tentacle_a_02, bone2Name=dyn_tentacle_a_01, swing1LimitDegrees=45, swing2LimitDegrees=45, twistLimitDegrees=45}
 {bone1Name=dyn_tentacle_a_03, bone2Name=dyn_tentacle_a_02, swing1LimitDegrees=45, swing2LimitDegrees=45, twistLimitDegrees=45}
 {bone1Name=dyn_tentacle_b_01, bone2Name=dyn_body_main, swing1LimitDegrees=45, swing2LimitDegrees=45, twistLimitDegrees=45}
 {bone1Name=dyn_tentacle_b_02, bone2Name=dyn_tentacle_b_01, swing1LimitDegrees=45, swing2LimitDegrees=45, twistLimitDegrees=45}
 {bone1Name=dyn_tentacle_b_03, bone2Name=dyn_tentacle_b_02, swing1LimitDegrees=45, swing2LimitDegrees=45, twistLimitDegrees=45}
 {bone1Name=dyn_tentacle_c_01, bone2Name=dyn_body_main, swing1LimitDegrees=45, swing2LimitDegrees=45, twistLimitDegrees=45}
 {bone1Name=dyn_tentacle_c_02, bone2Name=dyn_tentacle_c_01, swing1LimitDegrees=45, swing2LimitDegrees=45, twistLimitDegrees=45}
 {bone1Name=dyn_tentacle_d_01, bone2Name=dyn_body_main, swing1LimitDegrees=45, swing2LimitDegrees=45, twistLimitDegrees=45}
 {bone1Name=dyn_tentacle_d_02, bone2Name=dyn_tentacle_d_01, swing1LimitDegrees=45, swing2LimitDegrees=45, twistLimitDegrees=45}
 {bone1Name=dyn_tentacle_e_01, bone2Name=dyn_body_main, swing1LimitDegrees=45, swing2LimitDegrees=45, twistLimitDegrees=45}
 {bone1Name=dyn_tentacle_e_02, bone2Name=dyn_tentacle_e_01, swing1LimitDegrees=45, swing2LimitDegrees=45, twistLimitDegrees=45}
 {bone1Name=dyn_tentacle_f_01, bone2Name=dyn_body_main, swing1LimitDegrees=45, swing2LimitDegrees=45, twistLimitDegrees=45}
 {bone1Name=dyn_tentacle_f_02, bone2Name=dyn_tentacle_f_01, swing1LimitDegrees=45, swing2LimitDegrees=45, twistLimitDegrees=45}
 {bone1Name=dyn_tentacle_g_01, bone2Name=dyn_body_main, swing1LimitDegrees=45, swing2LimitDegrees=45, twistLimitDegrees=45}
 {bone1Name=dyn_tentacle_g_02, bone2Name=dyn_tentacle_g_01, swing1LimitDegrees=45, swing2LimitDegrees=45, twistLimitDegrees=45}
 {bone1Name=dyn_tentacle_h_01, bone2Name=dyn_body_main, swing1LimitDegrees=45, swing2LimitDegrees=45, twistLimitDegrees=45}
 {bone1Name=dyn_tentacle_h_02, bone2Name=dyn_tentacle_h_01, swing1LimitDegrees=45, swing2LimitDegrees=45, twistLimitDegrees=45}
 {bone1Name=dyn_tentacle_i_01, bone2Name=dyn_body_main, swing1LimitDegrees=45, swing2LimitDegrees=45, twistLimitDegrees=45}
 {bone1Name=dyn_tentacle_i_02, bone2Name=dyn_tentacle_i_01, swing1LimitDegrees=45, swing2LimitDegrees=45, twistLimitDegrees=45}
 {bone1Name=dyn_tentacle_j_01, bone2Name=dyn_body_main, swing1LimitDegrees=45, swing2LimitDegrees=45, twistLimitDegrees=45}
 {bone1Name=dyn_tentacle_j_02, bone2Name=dyn_tentacle_j_01, swing1LimitDegrees=45, swing2LimitDegrees=45, twistLimitDegrees=45}
 {bone1Name=dyn_tentacle_j_03, bone2Name=dyn_tentacle_j_02, swing1LimitDegrees=45, swing2LimitDegrees=45, twistLimitDegrees=45}
 {bone1Name=dyn_tentacle_k_01, bone2Name=dyn_body_main, swing1LimitDegrees=45, swing2LimitDegrees=45, twistLimitDegrees=45}
 {bone1Name=dyn_tentacle_k_02, bone2Name=dyn_tentacle_k_01, swing1LimitDegrees=45, swing2LimitDegrees=45, twistLimitDegrees=45}
]
