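# Physics asset for the dynamic body of the GummiMale1H pickaxe cosmetic, defining capsule bodies and joints for its dyn_ bones.
# A small main body anchors a five-link upper chain and a single lower segment, each link using an identical capsule.
# Constraints join each link to its parent with moderate swing limits and very tight twist, so the segments can bend but barely twist.
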PhysicsAsset /BRCosmetics/Weapons/FORT_Melee/Pickaxe_GummiMale1H/Meshes/Gummi_Male_1h_Axe_Dyn_body
bodies {
 dyn_body_main [{shapeType=Capsule, length=1, radius=1}]
 dyn_lw_body_01 [{shapeType=Capsule, length=6, radius=4, center=<3 0 0>, rotation=<rot 90 0 0>}]
 dyn_up_body_01 [{shapeType=Capsule, length=6, radius=4, center=<3 0 0>, rotation=<rot 90 0 0>}]
 dyn_up_body_02 [{shapeType=Capsule, length=6, radius=4, center=<3 0 0>, rotation=<rot 90 0 0>}]
 dyn_up_body_03 [{shapeType=Capsule, length=6, radius=4, center=<3 0 0>, rotation=<rot 90 0 0>}]
 dyn_up_body_04 [{shapeType=Capsule, length=6, radius=4, center=<3 0 0>, rotation=<rot 90 0 0>}]
 dyn_up_body_05 [{shapeType=Capsule, length=6, radius=4, center=<3 0 0>, rotation=<rot 90 0 0>}]
}
constraints [
 {bone1Name=dyn_lw_body_01, bone2Name=dyn_body_main, swing1LimitDegrees=30, swing2LimitDegrees=30, twistLimitDegrees=2}
 {bone1Name=dyn_up_body_01, bone2Name=dyn_body_main, swing1LimitDegrees=30, swing2LimitDegrees=30, twistLimitDegrees=2}
 {bone1Name=dyn_up_body_02, bone2Name=dyn_up_body_01, swing1LimitDegrees=30, swing2LimitDegrees=30, twistLimitDegrees=2}
 {bone1Name=dyn_up_body_03, bone2Name=dyn_up_body_02, swing1LimitDegrees=30, swing2LimitDegrees=30, twistLimitDegrees=2}
 {bone1Name=dyn_up_body_04, bone2Name=dyn_up_body_03, swing1LimitDegrees=30, swing2LimitDegrees=30, twistLimitDegrees=2}
 {bone1Name=dyn_up_body_05, bone2Name=dyn_up_body_04, swing1LimitDegrees=30, swing2LimitDegrees=30, twistLimitDegrees=2}
]
