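# Animation Blueprint for the Gummi male one-handed axe pickaxe cosmetic, adding physics wobble and procedural fin motion.
# Two RigidBody nodes with separate physics assets simulate in dyn_body_main space under reduced gravity.
# Each update, phase-offset sine waves from game time set four rotators, applied additively to mirrored left and right fin bones for a rippling wave.
# The pose then runs through the axe's Control Rig before output.
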AnimBlueprint /BRCosmetics/Weapons/FORT_Melee/Pickaxe_GummiMale1H/Meshes/Gummi_Male_1h_Axe_AnimBP
target_skeleton=&/Gummi_Male_1h_Axe_Skeleton
graphs [
 AnimGraph {nodes [
   n1 = RigidBody(ComponentPose=Local To Component(LocalPose=Input Pose())) {
    bClampLinearTranslationLimitToRefPose=true
    bOverrideWorldGravity=true
    overridePhysicsAsset=&/Gummi_Male_1h_Axe_Dyn
    simulationSpace=BaseBoneSpace
    alphaBoolBlend {blendOption=Linear}
    baseBoneRef {boneName=dyn_body_main}
    componentAppliedLinearAccClamp=<3000 3000 3000>
    componentLinearAccScale=<0.15 0.3 0.3>
    componentLinearVelScale=<0.5 0.5 0.5>
    overrideWorldGravity {z=-100}
   }
   n2 = RigidBody(ComponentPose=@n1) {
    bClampLinearTranslationLimitToRefPose=true
    bOverrideWorldGravity=true
    overridePhysicsAsset=&/Gummi_Male_1h_Axe_Dyn_body
    simulationSpace=BaseBoneSpace
    alphaBoolBlend {blendOption=Linear}
    baseBoneRef {boneName=dyn_body_main}
    componentAppliedLinearAccClamp=<3000 3000 3000>
    componentLinearAccScale=<0.5 0.7 0.7>
    componentLinearVelScale=<0.2 0.2 0.2>
    overrideWorldGravity {z=-100}
   }
   n3 = "Transform (Modify) Bone - Bone: dyn_fin_r_01"(ComponentPose=@n2, Alpha=1, Translation="0,0,0", Rotation=Get fin pins 01() -> (fin pins 01:Rotator), Scale="1,1,1") {rotationMode=BMM_Additive, rotationSpace=BCS_BoneSpace, alphaBoolBlend {blendOption=Linear}, boneToModify {boneName=dyn_fin_r_01}}
   n4 = "Transform (Modify) Bone - Bone: dyn_fin_r_02"(ComponentPose=@n3, Alpha=1, Translation="0,0,0", Rotation=Get fin pins 02() -> (fin pins 02:Rotator), Scale="1,1,1") {rotationMode=BMM_Additive, rotationSpace=BCS_BoneSpace, alphaBoolBlend {blendOption=Linear}, boneToModify {boneName=dyn_fin_r_02}}
   n5 = "Transform (Modify) Bone - Bone: dyn_fin_r_04"(ComponentPose=@n4, Alpha=1, Translation="0,0,0", Rotation=Get fin pins 03() -> (fin pins 03:Rotator), Scale="1,1,1") {rotationMode=BMM_Additive, rotationSpace=BCS_BoneSpace, alphaBoolBlend {blendOption=Linear}, boneToModify {boneName=dyn_fin_r_04}}
   n6 = "Transform (Modify) Bone - Bone: dyn_fin_r_05"(ComponentPose=@n5, Alpha=1, Translation="0,0,0", Rotation=Get fin pins 04() -> (fin pins 04:Rotator), Scale="1,1,1") {rotationMode=BMM_Additive, rotationSpace=BCS_BoneSpace, alphaBoolBlend {blendOption=Linear}, boneToModify {boneName=dyn_fin_r_05}}
   n7 = "Transform (Modify) Bone - Bone: dyn_fin_l_01"(ComponentPose=@n6, Alpha=1, Translation="0,0,0", Rotation=Get fin pins 01(), Scale="1,1,1") {rotationMode=BMM_Additive, rotationSpace=BCS_BoneSpace, alphaBoolBlend {blendOption=Linear}, boneToModify {boneName=dyn_fin_l_01}}
   n8 = "Transform (Modify) Bone - Bone: dyn_fin_l_02"(ComponentPose=@n7, Alpha=1, Translation="0,0,0", Rotation=Get fin pins 02(), Scale="1,1,1") {rotationMode=BMM_Additive, rotationSpace=BCS_BoneSpace, alphaBoolBlend {blendOption=Linear}, boneToModify {boneName=dyn_fin_l_02}}
   n9 = "Transform (Modify) Bone - Bone: dyn_fin_l_04"(ComponentPose=@n8, Alpha=1, Translation="0,0,0", Rotation=Get fin pins 03(), Scale="1,1,1") {rotationMode=BMM_Additive, rotationSpace=BCS_BoneSpace, alphaBoolBlend {blendOption=Linear}, boneToModify {boneName=dyn_fin_l_04}}
   n10 = "Transform (Modify) Bone - Bone: dyn_fin_l_05"(ComponentPose=@n9, Alpha=1, Translation="0,0,0", Rotation=Get fin pins 04(), Scale="1,1,1") {rotationMode=BMM_Additive, rotationSpace=BCS_BoneSpace, alphaBoolBlend {blendOption=Linear}, boneToModify {boneName=dyn_fin_l_05}}
   n11 = Control Rig(Source=Component To Local(ComponentPose=@n10)) {
    alphaBoolBlend {blendOption=Linear}
    controlRigAssetReference {blueprintRigClass=&/Gummi_Male_1h_Axe_CtrlRig.Gummi_Male_1h_Axe_CtrlRig_C}
   }
   Output Pose(Result=@n11) {layerGroup=None}
  ]}
 EventGraph {nodes [
   Event BlueprintUpdateAnimation()
   n1 = float * float(A=GetGameTimeInSeconds(), B=100)
   n2 = "Sin (Degrees)"!(A:double=@n1) -> (ReturnValue:double)
   n3 = MapRangeClamped(Value=@n2, InRangeA=-1, InRangeB=1, OutRangeA=-8, OutRangeB=8)
   > Set fin pins 01(fin pins 01:Rotator=MakeRotator(Roll=@n3, Pitch=MapRangeClamped(Value=@n2, InRangeA=-1, InRangeB=1, OutRangeA=-30, OutRangeB=30), Yaw=@n3)) -> (Output_Get:Rotator)
   n4 = "Sin (Degrees)"(A=float + float(A=@n1, B=-60) {is=title K2Node_CommutativeAssociativeBinaryOperator})
   n5 = MapRangeClamped(Value=@n4, InRangeA=-1, InRangeB=1, OutRangeA=-8, OutRangeB=8)
   > Set fin pins 02(fin pins 02:Rotator=MakeRotator(Roll=@n5, Pitch=MapRangeClamped(Value=@n4, InRangeA=-1, InRangeB=1, OutRangeA=-30, OutRangeB=30), Yaw=@n5)) -> (Output_Get:Rotator)
   n6 = "Sin (Degrees)"(A=float + float(A=@n1, B=-120))
   n7 = MapRangeClamped(Value=@n6, InRangeA=-1, InRangeB=1, OutRangeA=-8, OutRangeB=8)
   > Set fin pins 03(fin pins 03:Rotator=MakeRotator(Roll=@n7, Pitch=MapRangeClamped(Value=@n6, InRangeA=-1, InRangeB=1, OutRangeA=-30, OutRangeB=30), Yaw=@n7)) -> (Output_Get:Rotator)
   n8 = "Sin (Degrees)"(A=float + float(A=@n1, B=-180))
   n9 = MapRangeClamped(Value=@n8, InRangeA=-1, InRangeB=1, OutRangeA=-8, OutRangeB=8)
   > Set fin pins 04(fin pins 04:Rotator=MakeRotator(Roll=@n9, Pitch=MapRangeClamped(Value=@n8, InRangeA=-1, InRangeB=1, OutRangeA=-30, OutRangeB=30), Yaw=@n9)) -> (Output_Get:Rotator)
   TryGetPawnOwner()
  ]}
]
variables [
 fin pins 01 {type=Rotator}
 fin pins 02 {type=Rotator}
 fin pins 03 {type=Rotator}
 fin pins 04 {type=Rotator}
]
